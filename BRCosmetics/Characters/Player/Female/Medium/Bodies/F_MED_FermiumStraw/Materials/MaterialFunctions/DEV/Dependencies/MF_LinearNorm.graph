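# MF_LinearNorm is a material function that linearly normalizes a scalar Input from a Range Min to Range Max interval into 0–1 space.
# It subtracts Range Min from Input and divides by the span between Range Max and Range Min.
# The result is not clamped, so values outside the range map below 0 or above 1.
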MaterialFunction /BRCosmetics/Characters/Player/Female/Medium/Bodies/F_MED_FermiumStraw/Materials/MaterialFunctions/DEV/Dependencies/MF_LinearNorm
graph {nodes [
  n1 = FunctionInput() {inputName=Input, inputType=FunctionInput_Scalar}
  n2 = FunctionInput() {inputName=Range Min, inputType=FunctionInput_Scalar, sortPriority=10}
  n3 = FunctionInput() {inputName=Range Max, inputType=FunctionInput_Scalar, sortPriority=20, previewValue=<1 0 0 1>}
  FunctionOutput(None=Divide(A=Subtract(A=@n1, B=@n2), B=Subtract(A=@n3, B=@n2))) {description=Result is NOT clamped}
 ]}
