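# Material function that turns a Vector4 signal into a Vector3 debug visualization, located in the FermiumStraw character's DEV material dependencies.
# Switches swizzle any RGBA channels into one (greyscale), two (RG) or three (RGB) displayed channels.
# Per-channel sign inversion, domain change, one-minus, MF_LinearNorm range remap, offset, multiply, power and tint adjust the signal, with a static Enable Remaps switch.
# Rounded, saturated inputs act as per-channel toggles, including optional sRGB-to-linear output conversion.
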MaterialFunction /BRCosmetics/Characters/Player/Female/Medium/Bodies/F_MED_FermiumStraw/Materials/MaterialFunctions/DEV/Dependencies/MF_DebugSignal
graph {expressions {
  MaterialExpressionAdd_3 {
   class=MaterialExpressionAdd
   inputs [
    A {from=MaterialExpressionNamedRerouteUsage_5}
    B {from=MaterialExpressionFunctionInput_5}
   ]
  }
  MaterialExpressionClamp_0 {
   class=MaterialExpressionClamp
   inputs [
    None {from=MaterialExpressionFunctionInput_22, output=""}
    Min {from="", output=""}
    Max {from="", output=""}
   ]
   properties {minDefault=-1}
  }
  MaterialExpressionClamp_4 {
   class=MaterialExpressionClamp
   inputs [
    None {from=MaterialExpressionRound_4, output=""}
    Min {from="", output=""}
    Max {from="", output=""}
   ]
   properties {maxDefault=3}
  }
  MaterialExpressionClamp_5 {
   class=MaterialExpressionClamp
   inputs [
    None {from=MaterialExpressionRound_5, output=""}
    Min {from="", output=""}
    Max {from="", output=""}
   ]
   properties {maxDefault=3}
  }
  MaterialExpressionClamp_6 {
   class=MaterialExpressionClamp
   inputs [
    None {from=MaterialExpressionRound_6, output=""}
    Min {from="", output=""}
    Max {from="", output=""}
   ]
   properties {maxDefault=3}
  }
  MaterialExpressionClamp_7 {
   class=MaterialExpressionClamp
   inputs [
    None {from=MaterialExpressionRound_7, output=""}
    Min {from="", output=""}
    Max {from="", output=""}
   ]
   properties {maxDefault=2}
  }
  MaterialExpressionConstantBiasScale_0 {class=MaterialExpressionConstantBiasScale, inputs [None {from=MaterialExpressionNamedRerouteUsage_26}]}
  MaterialExpressionConstantBiasScale_1 {class=MaterialExpressionConstantBiasScale, inputs [None {from=MaterialExpressionNamedRerouteUsage_19}], properties {bias=-0.5, scale=2}}
  MaterialExpressionConvert_16 {
   class=MaterialExpressionConvert
   inputs [None {from=MaterialExpressionNamedRerouteUsage_2}]
   properties {
    nodeName=Break Float3
    convertInputs [{
      type=Vector3
      expressionInput {
       expression=&self:MaterialExpressionNamedRerouteUsage_2
       inputName=None
       mask=0
       maskA=0
       maskB=0
       maskG=0
       maskR=0
       outputIndex=0
      }
     }]
    convertOutputs [{type=Scalar}, {type=Scalar}, {type=Scalar}]
   }
  }
  MaterialExpressionConvert_21 {
   class=MaterialExpressionConvert
   inputs [
    None {from=MaterialExpressionMaterialFunctionCall_0, output=Result}
    None {from=MaterialExpressionMaterialFunctionCall_8, output=Result}
    None {from=MaterialExpressionMaterialFunctionCall_7, output=Result}
   ]
   properties {
    nodeName=Make Vector3
    convertInputs [
     {
      type=Scalar
      expressionInput {
       expression=&self:MaterialExpressionMaterialFunctionCall_0
       inputName=None
       mask=0
       maskA=0
       maskB=0
       maskG=0
       maskR=0
       outputIndex=0
      }
     }
     {
      type=Scalar
      expressionInput {
       expression=&self:MaterialExpressionMaterialFunctionCall_8
       inputName=None
       mask=0
       maskA=0
       maskB=0
       maskG=0
       maskR=0
       outputIndex=0
      }
     }
     {
      type=Scalar
      expressionInput {
       expression=&self:MaterialExpressionMaterialFunctionCall_7
       inputName=None
       mask=0
       maskA=0
       maskB=0
       maskG=0
       maskR=0
       outputIndex=0
      }
     }
    ]
    convertOutputs [{type=Vector3}]
   }
  }
  MaterialExpressionConvert_22 {
   class=MaterialExpressionConvert
   inputs [None {from=MaterialExpressionFunctionInput_15}]
   properties {
    nodeName=Break Float3
    convertInputs [{
      type=Vector3
      expressionInput {
       expression=&self:MaterialExpressionFunctionInput_15
       inputName=None
       mask=0
       maskA=0
       maskB=0
       maskG=0
       maskR=0
       outputIndex=0
      }
     }]
    convertOutputs [{type=Scalar}, {type=Scalar}, {type=Scalar}]
   }
  }
  MaterialExpressionConvert_23 {
   class=MaterialExpressionConvert
   inputs [None {from=MaterialExpressionFunctionInput_16}]
   properties {
    nodeName=Break Float3
    convertInputs [{
      type=Vector3
      expressionInput {
       expression=&self:MaterialExpressionFunctionInput_16
       inputName=None
       mask=0
       maskA=0
       maskB=0
       maskG=0
       maskR=0
       outputIndex=0
      }
     }]
    convertOutputs [{type=Scalar}, {type=Scalar}, {type=Scalar}]
   }
  }
  MaterialExpressionConvert_5 {
   class=MaterialExpressionConvert
   inputs [None {from=MaterialExpressionFunctionInput_0}]
   properties {
    nodeName=Break Float4
    convertInputs [{
      type=Vector4
      defaultValue=<rgba 0 0 0 1>
      expressionInput {
       expression=&self:MaterialExpressionFunctionInput_0
       inputName=None
       mask=0
       maskA=0
       maskB=0
       maskG=0
       maskR=0
       outputIndex=0
      }
     }]
    convertOutputs [{type=Scalar}, {type=Scalar}, {type=Scalar}, {type=Scalar}]
   }
  }
  MaterialExpressionConvert_6 {
   class=MaterialExpressionConvert
   inputs [
    None {from=MaterialExpressionSwitch_5}
    None {from=MaterialExpressionSwitch_5}
    None {from=MaterialExpressionSwitch_5}
   ]
   properties {
    nodeName=Make Vector3
    convertInputs [
     {
      type=Scalar
      expressionInput {
       expression=&self:MaterialExpressionSwitch_5
       inputName=None
       mask=0
       maskA=0
       maskB=0
       maskG=0
       maskR=0
       outputIndex=0
      }
     }
     {
      type=Scalar
      expressionInput {
       expression=&self:MaterialExpressionSwitch_5
       inputName=None
       mask=0
       maskA=0
       maskB=0
       maskG=0
       maskR=0
       outputIndex=0
      }
     }
     {
      type=Scalar
      expressionInput {
       expression=&self:MaterialExpressionSwitch_5
       inputName=None
       mask=0
       maskA=0
       maskB=0
       maskG=0
       maskR=0
       outputIndex=0
      }
     }
    ]
    convertOutputs [{type=Vector3}]
   }
  }
  MaterialExpressionConvert_7 {
   class=MaterialExpressionConvert
   inputs [
    None {from=MaterialExpressionSwitch_5}
    None {from=MaterialExpressionSwitch_6}
    None {from=""}
   ]
   properties {
    nodeName=Make Vector3
    convertInputs [
     {
      type=Scalar
      expressionInput {
       expression=&self:MaterialExpressionSwitch_5
       inputName=None
       mask=0
       maskA=0
       maskB=0
       maskG=0
       maskR=0
       outputIndex=0
      }
     }
     {
      type=Scalar
      expressionInput {
       expression=&self:MaterialExpressionSwitch_6
       inputName=None
       mask=0
       maskA=0
       maskB=0
       maskG=0
       maskR=0
       outputIndex=0
      }
     }
     {type=Scalar, expressionInput {expression=null, inputName=None, mask=0, maskA=0, maskB=0, maskG=0, maskR=0, outputIndex=0}}
    ]
    convertOutputs [{type=Vector3}]
   }
  }
  MaterialExpressionConvert_8 {
   class=MaterialExpressionConvert
   inputs [
    None {from=MaterialExpressionSwitch_5}
    None {from=MaterialExpressionSwitch_6}
    None {from=MaterialExpressionSwitch_7}
   ]
   properties {
    nodeName=Make Vector3
    convertInputs [
     {
      type=Scalar
      expressionInput {
       expression=&self:MaterialExpressionSwitch_5
       inputName=None
       mask=0
       maskA=0
       maskB=0
       maskG=0
       maskR=0
       outputIndex=0
      }
     }
     {
      type=Scalar
      expressionInput {
       expression=&self:MaterialExpressionSwitch_6
       inputName=None
       mask=0
       maskA=0
       maskB=0
       maskG=0
       maskR=0
       outputIndex=0
      }
     }
     {
      type=Scalar
      expressionInput {
       expression=&self:MaterialExpressionSwitch_7
       inputName=None
       mask=0
       maskA=0
       maskB=0
       maskG=0
       maskR=0
       outputIndex=0
      }
     }
    ]
    convertOutputs [{type=Vector3}]
   }
  }
  MaterialExpressionFunctionInput_0 {
   class=MaterialExpressionFunctionInput
   inputs [Preview {from="", output=""}]
   properties {inputName=Signal In, inputType=FunctionInput_Vector4, previewValue=<1 0 0 1>}
  }
  MaterialExpressionFunctionInput_1 {
   class=MaterialExpressionFunctionInput
   inputs [Preview {from="", output=""}]
   properties {
    bUsePreviewValueAsDefault=true
    desc="\r\n1, 2 or 3\r\n(R, RG, RGB)"
    description="1, 2 or 3 (R, RG, RGB)"
    inputName=Swizzle - Num Debug Channels
    inputType=FunctionInput_Scalar
    sortPriority=20
    previewValue=<3 0 0 1>
   }
  }
  MaterialExpressionFunctionInput_15 {
   class=MaterialExpressionFunctionInput
   inputs [Preview {from="", output=""}]
   properties {bUsePreviewValueAsDefault=true, inputName=Remap - Range OldMin, sortPriority=150}
  }
  MaterialExpressionFunctionInput_16 {
   class=MaterialExpressionFunctionInput
   inputs [Preview {from="", output=""}]
   properties {bUsePreviewValueAsDefault=true, inputName=Remap - Range OldMax, sortPriority=160, previewValue=<1 1 1 1>}
  }
  MaterialExpressionFunctionInput_17 {
   class=MaterialExpressionFunctionInput
   inputs [Preview {from="", output=""}]
   properties {bUsePreviewValueAsDefault=true, inputName=//------ PER-CHANNEL REMAPS ------//, inputType=FunctionInput_Scalar, sortPriority=100}
  }
  MaterialExpressionFunctionInput_18 {
   class=MaterialExpressionFunctionInput
   inputs [Preview {from="", output=""}]
   properties {bUsePreviewValueAsDefault=true, inputName=//------ CHANNELS SWIZZLE ------//, inputType=FunctionInput_Scalar, sortPriority=10}
  }
  MaterialExpressionFunctionInput_19 {
   class=MaterialExpressionFunctionInput
   inputs [Preview {from=MaterialExpressionStaticBool_0, output=""}]
   properties {bUsePreviewValueAsDefault=true, inputName=Enable Remaps, inputType=FunctionInput_StaticBool, sortPriority=110}
  }
  MaterialExpressionFunctionInput_2 {
   class=MaterialExpressionFunctionInput
   inputs [Preview {from="", output=""}]
   properties {
    bUsePreviewValueAsDefault=true
    desc="\r\n0: R. 1: G, 2: B, 3: A"
    inputName=Swizzle - R Channel Out
    inputType=FunctionInput_Scalar
    sortPriority=30
   }
  }
  MaterialExpressionFunctionInput_21 {
   class=MaterialExpressionFunctionInput
   inputs [Preview {from="", output=""}]
   properties {bUsePreviewValueAsDefault=true, inputName=Remap - One Minus, sortPriority=140, previewValue=<0 0 0 0>}
  }
  MaterialExpressionFunctionInput_22 {
   class=MaterialExpressionFunctionInput
   inputs [Preview {from="", output=""}]
   properties {
    bUsePreviewValueAsDefault=true
    desc="-1 :  [0, 1] to [-1, 1]\r\n0  : no change\r\n1 : [-1, 1] to [0, 1]\r\n"
    description="-1 :  [0, 1] to [-1, 1];  0  : no change;  1 : [-1, 1] to [0, 1]"
    inputName="Remap - Change Domain (see tooltip)"
    sortPriority=130
    previewValue=<0 0 0 0>
   }
  }
  MaterialExpressionFunctionInput_23 {
   class=MaterialExpressionFunctionInput
   inputs [Preview {from="", output=""}]
   properties {bUsePreviewValueAsDefault=true, inputName=Remap - Invert Sign, sortPriority=120, previewValue=<0 0 0 0>}
  }
  MaterialExpressionFunctionInput_24 {
   class=MaterialExpressionFunctionInput
   inputs [Preview {from="", output=""}]
   properties {bUsePreviewValueAsDefault=true, inputName=//------ PER-CHANNEL OUTPUT ------//, inputType=FunctionInput_Scalar, sortPriority=300}
  }
  MaterialExpressionFunctionInput_25 {
   class=MaterialExpressionFunctionInput
   inputs [Preview {from="", output=""}]
   properties {bUsePreviewValueAsDefault=true, inputName=Output - sRGB2Linear, sortPriority=310, previewValue=<0 0 0 0>}
  }
  MaterialExpressionFunctionInput_3 {
   class=MaterialExpressionFunctionInput
   inputs [Preview {from="", output=""}]
   properties {
    bUsePreviewValueAsDefault=true
    desc="\r\n0: R. 1: G, 2: B, 3: A"
    inputName=Swizzle - G Channel Out
    inputType=FunctionInput_Scalar
    sortPriority=40
    previewValue=<1 0 0 1>
   }
  }
  MaterialExpressionFunctionInput_4 {
   class=MaterialExpressionFunctionInput
   inputs [Preview {from="", output=""}]
   properties {
    bUsePreviewValueAsDefault=true
    desc="\r\n0: R. 1: G, 2: B, 3: A"
    inputName=Swizzle - B Channel Out
    inputType=FunctionInput_Scalar
    sortPriority=50
    previewValue=<2 0 0 1>
   }
  }
  MaterialExpressionFunctionInput_5 {
   class=MaterialExpressionFunctionInput
   inputs [Preview {from="", output=""}]
   properties {bUsePreviewValueAsDefault=true, inputName=Maths - Offset, sortPriority=210, previewValue=<0 0 0 0>}
  }
  MaterialExpressionFunctionInput_6 {
   class=MaterialExpressionFunctionInput
   inputs [Preview {from="", output=""}]
   properties {bUsePreviewValueAsDefault=true, inputName=Maths - Multiply, sortPriority=220, previewValue=<1 1 1 1>}
  }
  MaterialExpressionFunctionInput_7 {
   class=MaterialExpressionFunctionInput
   inputs [Preview {from="", output=""}]
   properties {bUsePreviewValueAsDefault=true, inputName=Maths - Power, sortPriority=230, previewValue=<1 1 1 1>}
  }
  MaterialExpressionFunctionInput_8 {
   class=MaterialExpressionFunctionInput
   inputs [Preview {from="", output=""}]
   properties {bUsePreviewValueAsDefault=true, inputName=Color - RGB Tint, sortPriority=240, previewValue=<1 1 1 1>}
  }
  MaterialExpressionFunctionOutput_0 {
   class=MaterialExpressionFunctionOutput
   inputs [None {from=MaterialExpressionNamedRerouteUsage_15, output=""}]
   properties {outputName="Debug (V3)"}
  }
  MaterialExpressionIf_0 {
   class=MaterialExpressionIf
   inputs [
    A {from=MaterialExpressionRound_2, output=""}
    B {from="", output=""}
    "A > B" {from=MaterialExpressionConstantBiasScale_0, output=""}
    "A == B" {from="", output=""}
    "A < B" {from=MaterialExpressionIf_1, output=""}
   ]
   properties {constB=1}
  }
  MaterialExpressionIf_1 {
   class=MaterialExpressionIf
   inputs [
    A {from=MaterialExpressionRound_2, output=""}
    B {from="", output=""}
    "A > B" {from=MaterialExpressionNamedRerouteUsage_27, output=""}
    "A == B" {from="", output=""}
    "A < B" {from=MaterialExpressionConstantBiasScale_1, output=""}
   ]
   properties {}
  }
  MaterialExpressionLinearInterpolate_0 {
   class=MaterialExpressionLinearInterpolate
   inputs [
    A {from=MaterialExpressionNamedRerouteUsage_4}
    B {from=MaterialExpressionMultiply_0}
    Alpha {from=MaterialExpressionReroute_13}
   ]
  }
  MaterialExpressionLinearInterpolate_4 {
   class=MaterialExpressionLinearInterpolate
   inputs [
    A {from=MaterialExpressionNamedRerouteUsage_25}
    B {from=MaterialExpressionOneMinus_3}
    Alpha {from=MaterialExpressionSaturate_8}
   ]
  }
  MaterialExpressionLinearInterpolate_6 {
   class=MaterialExpressionLinearInterpolate
   inputs [
    A {from=MaterialExpressionNamedRerouteUsage_23}
    B {from=MaterialExpressionMaterialFunctionCall_6, output="Vector3 Out (V3)"}
    Alpha {from=MaterialExpressionSaturate_9}
   ]
  }
  MaterialExpressionMaterialFunctionCall_0 {
   class=MaterialExpressionMaterialFunctionCall
   inputs [
    Input {from=""}
    Range Min {from=""}
    Range Max {from=""}
   ]
   properties {materialFunction=&/MF_LinearNorm}
  }
  MaterialExpressionMaterialFunctionCall_6 {
   class=MaterialExpressionMaterialFunctionCall
   inputs [
    Scalar In {from=""}
    Vector2 In {from=""}
    Vector3 In {from=""}
    Vector4 In {from=""}
   ]
   properties {materialFunction=&/MF_sRGBToLinear_RGBA}
  }
  MaterialExpressionMaterialFunctionCall_7 {
   class=MaterialExpressionMaterialFunctionCall
   inputs [
    Input {from=""}
    Range Min {from=""}
    Range Max {from=""}
   ]
   properties {materialFunction=&MF_LinearNorm}
  }
  MaterialExpressionMaterialFunctionCall_8 {
   class=MaterialExpressionMaterialFunctionCall
   inputs [
    Input {from=""}
    Range Min {from=""}
    Range Max {from=""}
   ]
   properties {materialFunction=&MF_LinearNorm}
  }
  MaterialExpressionMultiply_0 {
   class=MaterialExpressionMultiply
   inputs [
    A {from=MaterialExpressionNamedRerouteUsage_4}
    B {from=""}
   ]
   properties {constB=-1}
  }
  MaterialExpressionMultiply_4 {
   class=MaterialExpressionMultiply
   inputs [
    A {from=MaterialExpressionNamedRerouteUsage_6}
    B {from=MaterialExpressionFunctionInput_6}
   ]
  }
  MaterialExpressionMultiply_5 {
   class=MaterialExpressionMultiply
   inputs [
    A {from=MaterialExpressionNamedRerouteUsage_8}
    B {from=MaterialExpressionSaturate_1}
   ]
  }
  MaterialExpressionNamedRerouteDeclaration_0 {
   class=MaterialExpressionNamedRerouteDeclaration
   inputs [None {from=MaterialExpressionClamp_4}]
   properties {name="R Channel Select (S)", nodeColor=<rgba 0.5469 0 1 1>}
  }
  MaterialExpressionNamedRerouteDeclaration_1 {
   class=MaterialExpressionNamedRerouteDeclaration
   inputs [None {from=MaterialExpressionLinearInterpolate_4}]
   properties {name="Debug Signal - Post Invert (V3)", nodeColor=<rgba 1 0 0.7031 1>}
  }
  MaterialExpressionNamedRerouteDeclaration_10 {
   class=MaterialExpressionNamedRerouteDeclaration
   inputs [None {from=MaterialExpressionClamp_7}]
   properties {name="Num Channels Select (S)", nodeColor=<rgba 0.5469 0 1 1>}
  }
  MaterialExpressionNamedRerouteDeclaration_11 {
   class=MaterialExpressionNamedRerouteDeclaration
   inputs [None {from=MaterialExpressionSwitch_0}]
   properties {name="Debug Signal - Post Swizzle (V3)", nodeColor=<rgba 1 0 0.7031 1>}
  }
  MaterialExpressionNamedRerouteDeclaration_12 {
   class=MaterialExpressionNamedRerouteDeclaration
   inputs [None {from=MaterialExpressionMultiply_5}]
   properties {name="Debug Signal - Post Tint (V3)", nodeColor=<rgba 0.9453 0 1 1>}
  }
  MaterialExpressionNamedRerouteDeclaration_14 {
   class=MaterialExpressionNamedRerouteDeclaration
   inputs [None {from=MaterialExpressionLinearInterpolate_6}]
   properties {name="Debug Signal - Post sRGB2Linear (V3)", nodeColor=<rgba 1 0 0.7031 1>}
  }
  MaterialExpressionNamedRerouteDeclaration_16 {
   class=MaterialExpressionNamedRerouteDeclaration
   inputs [None {from=MaterialExpressionIf_0}]
   properties {name="Debug Signal - Post Remap Domain (V3)", nodeColor=<rgba 1 0 0.7031 1>}
  }
  MaterialExpressionNamedRerouteDeclaration_2 {
   class=MaterialExpressionNamedRerouteDeclaration
   inputs [None {from=MaterialExpressionConvert_21}]
   properties {name="Debug Signal - Post Remap (V3)", nodeColor=<rgba 1 0 0.7031 1>}
  }
  MaterialExpressionNamedRerouteDeclaration_3 {
   class=MaterialExpressionNamedRerouteDeclaration
   inputs [None {from=MaterialExpressionAdd_3}]
   properties {name="Debug Signal - Post Offset (V3)", nodeColor=<rgba 1 0 0.7031 1>}
  }
  MaterialExpressionNamedRerouteDeclaration_4 {
   class=MaterialExpressionNamedRerouteDeclaration
   inputs [None {from=MaterialExpressionMultiply_4}]
   properties {name="Debug Signal - Post Multiply (V3)", nodeColor=<rgba 1 0 0.7031 1>}
  }
  MaterialExpressionNamedRerouteDeclaration_5 {
   class=MaterialExpressionNamedRerouteDeclaration
   inputs [None {from=MaterialExpressionStaticSwitch_0}]
   properties {name="Debug Signal - Pre sRGB2Linear (V3)", nodeColor=<rgba 1 0 0.7031 1>}
  }
  MaterialExpressionNamedRerouteDeclaration_6 {
   class=MaterialExpressionNamedRerouteDeclaration
   inputs [None {from=MaterialExpressionPower_2}]
   properties {name="Debug Signal - Post Power (V3)", nodeColor=<rgba 1 0 0.7031 1>}
  }
  MaterialExpressionNamedRerouteDeclaration_7 {
   class=MaterialExpressionNamedRerouteDeclaration
   inputs [None {from=MaterialExpressionLinearInterpolate_0}]
   properties {name="Debug Signal - Post Invert Sign (V3)", nodeColor=<rgba 1 0 0.7031 1>}
  }
  MaterialExpressionNamedRerouteDeclaration_8 {
   class=MaterialExpressionNamedRerouteDeclaration
   inputs [None {from=MaterialExpressionClamp_5}]
   properties {name="G Channel Select (S)", nodeColor=<rgba 0.5469 0 1 1>}
  }
  MaterialExpressionNamedRerouteDeclaration_9 {
   class=MaterialExpressionNamedRerouteDeclaration
   inputs [None {from=MaterialExpressionClamp_6}]
   properties {name="B Channel Select (S)", nodeColor=<rgba 0.5469 0 1 1>}
  }
  MaterialExpressionNamedRerouteUsage_0 {class=MaterialExpressionNamedRerouteUsage}
  MaterialExpressionNamedRerouteUsage_10 {class=MaterialExpressionNamedRerouteUsage}
  MaterialExpressionNamedRerouteUsage_11 {class=MaterialExpressionNamedRerouteUsage}
  MaterialExpressionNamedRerouteUsage_13 {class=MaterialExpressionNamedRerouteUsage}
  MaterialExpressionNamedRerouteUsage_15 {class=MaterialExpressionNamedRerouteUsage}
  MaterialExpressionNamedRerouteUsage_16 {class=MaterialExpressionNamedRerouteUsage}
  MaterialExpressionNamedRerouteUsage_19 {class=MaterialExpressionNamedRerouteUsage}
  MaterialExpressionNamedRerouteUsage_2 {class=MaterialExpressionNamedRerouteUsage}
  MaterialExpressionNamedRerouteUsage_23 {class=MaterialExpressionNamedRerouteUsage}
  MaterialExpressionNamedRerouteUsage_25 {class=MaterialExpressionNamedRerouteUsage}
  MaterialExpressionNamedRerouteUsage_26 {class=MaterialExpressionNamedRerouteUsage}
  MaterialExpressionNamedRerouteUsage_27 {class=MaterialExpressionNamedRerouteUsage}
  MaterialExpressionNamedRerouteUsage_4 {class=MaterialExpressionNamedRerouteUsage}
  MaterialExpressionNamedRerouteUsage_5 {class=MaterialExpressionNamedRerouteUsage}
  MaterialExpressionNamedRerouteUsage_6 {class=MaterialExpressionNamedRerouteUsage}
  MaterialExpressionNamedRerouteUsage_7 {class=MaterialExpressionNamedRerouteUsage}
  MaterialExpressionNamedRerouteUsage_8 {class=MaterialExpressionNamedRerouteUsage}
  MaterialExpressionNamedRerouteUsage_9 {class=MaterialExpressionNamedRerouteUsage}
  MaterialExpressionOneMinus_3 {class=MaterialExpressionOneMinus, inputs [None {from=MaterialExpressionNamedRerouteUsage_25}]}
  MaterialExpressionPower_2 {
   class=MaterialExpressionPower
   inputs [
    Base {from=MaterialExpressionNamedRerouteUsage_7}
    Exp {from=MaterialExpressionFunctionInput_7}
   ]
  }
  MaterialExpressionReroute_13 {class=MaterialExpressionReroute, inputs [None {from=MaterialExpressionSaturate_3}]}
  MaterialExpressionRound_1 {class=MaterialExpressionRound, inputs [None {from=MaterialExpressionFunctionInput_23, output=""}], properties {}}
  MaterialExpressionRound_10 {class=MaterialExpressionRound, inputs [None {from=MaterialExpressionFunctionInput_25, output=""}], properties {}}
  MaterialExpressionRound_2 {class=MaterialExpressionRound, inputs [None {from=MaterialExpressionClamp_0, output=""}], properties {}}
  MaterialExpressionRound_4 {class=MaterialExpressionRound, inputs [None {from=MaterialExpressionFunctionInput_2, output=""}], properties {}}
  MaterialExpressionRound_5 {class=MaterialExpressionRound, inputs [None {from=MaterialExpressionFunctionInput_3, output=""}], properties {}}
  MaterialExpressionRound_6 {class=MaterialExpressionRound, inputs [None {from=MaterialExpressionFunctionInput_4, output=""}], properties {}}
  MaterialExpressionRound_7 {class=MaterialExpressionRound, inputs [None {from=MaterialExpressionSubtract_1, output=""}], properties {}}
  MaterialExpressionRound_9 {class=MaterialExpressionRound, inputs [None {from=MaterialExpressionFunctionInput_21, output=""}], properties {}}
  MaterialExpressionSaturate_1 {class=MaterialExpressionSaturate, inputs [None {from=MaterialExpressionFunctionInput_8}]}
  MaterialExpressionSaturate_3 {class=MaterialExpressionSaturate, inputs [None {from=MaterialExpressionRound_1}]}
  MaterialExpressionSaturate_8 {class=MaterialExpressionSaturate, inputs [None {from=MaterialExpressionRound_9}]}
  MaterialExpressionSaturate_9 {class=MaterialExpressionSaturate, inputs [None {from=MaterialExpressionRound_10}]}
  MaterialExpressionStaticBool_0 {class=MaterialExpressionStaticBool}
  MaterialExpressionStaticSwitch_0 {
   class=MaterialExpressionStaticSwitch
   inputs [
    True {from=MaterialExpressionNamedRerouteUsage_13, output=""}
    False {from=MaterialExpressionNamedRerouteUsage_16, output=""}
    Value {from=MaterialExpressionFunctionInput_19, output=""}
   ]
   properties {}
  }
  MaterialExpressionSubtract_1 {
   class=MaterialExpressionSubtract
   inputs [
    A {from=MaterialExpressionFunctionInput_1}
    B {from=""}
   ]
  }
  MaterialExpressionSwitch_0 {
   class=MaterialExpressionSwitch
   inputs [
    SwitchValue {from=MaterialExpressionNamedRerouteUsage_11, output=""}
    Default {from="", output=""}
    1Channel {from=MaterialExpressionConvert_6, output=""}
    2Channels {from=MaterialExpressionConvert_7, output=""}
    3Channels {from=MaterialExpressionConvert_8, output=""}
   ]
   properties {
    description=RGBChannels
    inputs [
     {
      inputName=1Channel
      input {
       expression=&self:MaterialExpressionConvert_6
       inputName=None
       mask=0
       maskA=0
       maskB=0
       maskG=0
       maskR=0
       outputIndex=0
      }
     }
     {
      inputName=2Channels
      input {
       expression=&self:MaterialExpressionConvert_7
       inputName=None
       mask=0
       maskA=0
       maskB=0
       maskG=0
       maskR=0
       outputIndex=0
      }
     }
     {
      inputName=3Channels
      input {
       expression=&self:MaterialExpressionConvert_8
       inputName=None
       mask=0
       maskA=0
       maskB=0
       maskG=0
       maskR=0
       outputIndex=0
      }
     }
    ]
   }
  }
  MaterialExpressionSwitch_5 {
   class=MaterialExpressionSwitch
   inputs [
    SwitchValue {from=MaterialExpressionNamedRerouteUsage_0, output=""}
    Default {from="", output=""}
    R {from=MaterialExpressionConvert_5, output=""}
    G {from=MaterialExpressionConvert_5, output=""}
    B {from=MaterialExpressionConvert_5, output=""}
    A {from=MaterialExpressionConvert_5, output=""}
   ]
   properties {
    description=RChannel
    inputs [
     {
      inputName=R
      input {
       expression=&self:MaterialExpressionConvert_5
       inputName=None
       mask=0
       maskA=0
       maskB=0
       maskG=0
       maskR=0
       outputIndex=0
      }
     }
     {
      inputName=G
      input {
       expression=&self:MaterialExpressionConvert_5
       inputName=None
       mask=0
       maskA=0
       maskB=0
       maskG=0
       maskR=0
       outputIndex=1
      }
     }
     {
      inputName=B
      input {
       expression=&self:MaterialExpressionConvert_5
       inputName=None
       mask=0
       maskA=0
       maskB=0
       maskG=0
       maskR=0
       outputIndex=2
      }
     }
     {
      inputName=A
      input {
       expression=&self:MaterialExpressionConvert_5
       inputName=None
       mask=0
       maskA=0
       maskB=0
       maskG=0
       maskR=0
       outputIndex=3
      }
     }
    ]
   }
  }
  MaterialExpressionSwitch_6 {
   class=MaterialExpressionSwitch
   inputs [
    SwitchValue {from=MaterialExpressionNamedRerouteUsage_9, output=""}
    Default {from="", output=""}
    R {from=MaterialExpressionConvert_5, output=""}
    G {from=MaterialExpressionConvert_5, output=""}
    B {from=MaterialExpressionConvert_5, output=""}
    A {from=MaterialExpressionConvert_5, output=""}
   ]
   properties {
    description=GChannel
    inputs [
     {
      inputName=R
      input {
       expression=&self:MaterialExpressionConvert_5
       inputName=None
       mask=0
       maskA=0
       maskB=0
       maskG=0
       maskR=0
       outputIndex=0
      }
     }
     {
      inputName=G
      input {
       expression=&self:MaterialExpressionConvert_5
       inputName=None
       mask=0
       maskA=0
       maskB=0
       maskG=0
       maskR=0
       outputIndex=1
      }
     }
     {
      inputName=B
      input {
       expression=&self:MaterialExpressionConvert_5
       inputName=None
       mask=0
       maskA=0
       maskB=0
       maskG=0
       maskR=0
       outputIndex=2
      }
     }
     {
      inputName=A
      input {
       expression=&self:MaterialExpressionConvert_5
       inputName=None
       mask=0
       maskA=0
       maskB=0
       maskG=0
       maskR=0
       outputIndex=3
      }
     }
    ]
   }
  }
  MaterialExpressionSwitch_7 {
   class=MaterialExpressionSwitch
   inputs [
    SwitchValue {from=MaterialExpressionNamedRerouteUsage_10, output=""}
    Default {from="", output=""}
    R {from=MaterialExpressionConvert_5, output=""}
    G {from=MaterialExpressionConvert_5, output=""}
    B {from=MaterialExpressionConvert_5, output=""}
    A {from=MaterialExpressionConvert_5, output=""}
   ]
   properties {
    description=BChannel
    inputs [
     {
      inputName=R
      input {
       expression=&self:MaterialExpressionConvert_5
       inputName=None
       mask=0
       maskA=0
       maskB=0
       maskG=0
       maskR=0
       outputIndex=0
      }
     }
     {
      inputName=G
      input {
       expression=&self:MaterialExpressionConvert_5
       inputName=None
       mask=0
       maskA=0
       maskB=0
       maskG=0
       maskR=0
       outputIndex=1
      }
     }
     {
      inputName=B
      input {
       expression=&self:MaterialExpressionConvert_5
       inputName=None
       mask=0
       maskA=0
       maskB=0
       maskG=0
       maskR=0
       outputIndex=2
      }
     }
     {
      inputName=A
      input {
       expression=&self:MaterialExpressionConvert_5
       inputName=None
       mask=0
       maskA=0
       maskB=0
       maskG=0
       maskR=0
       outputIndex=3
      }
     }
    ]
   }
  }
 }}
